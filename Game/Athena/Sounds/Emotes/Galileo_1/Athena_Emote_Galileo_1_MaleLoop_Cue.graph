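# Looping sound cue for the male variant of the Galileo_1 emote in Athena.
# A Mixer layers looping foley and ambient waves, boosting both inputs equally above unity.
# It uses the first-person emote music sound class and concurrency set, despite carrying no music-named wave.
# It also declares a send to the music audio-analysis submix.
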SoundCue /Game/Athena/Sounds/Emotes/Galileo_1/Athena_Emote_Galileo_1_MaleLoop_Cue
root=Mixer_2
nodes [
 {id=Mixer_2, children [WavePlayer_9, WavePlayer_10], properties {inputVolume [2,2]}}
 {id=WavePlayer_10, properties {bLooping=true, soundWaveAssetPtr=&/Emote_Galileo_1_Ambient_Loop}}
 {id=WavePlayer_9, properties {bLooping=true, soundWaveAssetPtr=&/Emote_Galileo_1_Foley_Loop}}
]
properties {
 maxDistance=2097152
 soundClassObject=/Game/Sounds/FortSoundClassesAndModes/Fort_Music_Emote_1P_PSM
 volumeMultiplier=1
 concurrencyOverrides {maxCount=3, resolutionRule=StopOldest}
 concurrencySet [/Game/Sounds/Concurrency_Rules/Athena_EmoteMusic_1P_Conc]
 soundSubmixSends [{minSendLevel=0, soundSubmix=&Fortnite_AudioAnalysis_Music}]
}
