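# Third-person looping sound cue for the male version of the Galileo_1 emote in Athena.
# A mixer layers two looping waves, an ambient loop and a foley loop, each with boosted input volume.
# It uses emote attenuation, the third-person emote music class and emote music concurrency, and keeps playing when silent.
# A submix send feeds the music audio analysis submix.
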SoundCue /Game/Athena/Sounds/Emotes/Galileo_1/SC_Athena_Emote_Galileo_1_MaleLoop_3p_Cue
root=Mixer_2
nodes [
 {id=Mixer_2, children [WavePlayer_9, WavePlayer_10], properties {inputVolume [2,2]}}
 {id=WavePlayer_10, properties {bLooping=true, soundWaveAssetPtr=&/Emote_Galileo_1_Ambient_Loop}}
 {id=WavePlayer_9, properties {bLooping=true, soundWaveAssetPtr=&/Emote_Galileo_1_Foley_Loop}}
]
properties {
 attenuationSettings=&Fort_Emotes_Athena_Att
 maxDistance=1600
 soundClassObject=/Game/Sounds/FortSoundClassesAndModes/Fort_Music_Emote_3P
 virtualizationMode=PlayWhenSilent
 volumeMultiplier=1
 concurrencyOverrides {maxCount=3, resolutionRule=StopOldest}
 concurrencySet [/Game/Sounds/Concurrency_Rules/Athena_EmoteMusic_Conc]
 soundSubmixSends [{minSendLevel=0, soundSubmix=&Fortnite_AudioAnalysis_Music}]
}
